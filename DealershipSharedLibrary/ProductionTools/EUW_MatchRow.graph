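# Editor utility widget for one result row in a material query tool, showing a match's name, path, reason, AI confidence score, node count and parameter count.
# PreConstruct fills the text fields from variables, or collapses the name and data block, buttons included, when NoResults is set.
# Buttons open the material or Niagara system through Python calls to material_query_euw.
# Helper functions return a red tint for no results and hide empty Niagara paths.
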EditorUtilityWidgetBlueprint /DealershipSharedLibrary/ProductionTools/EUW_MatchRow
graphs [
 EventGraph {nodes [
   "On Clicked (OpenButton)"() {is=title K2Node_ComponentBoundEvent}
   > ExecutePythonCommand(PythonCommand=Append!(A="import material_query_euw; material_query_euw.open_material(r'", B=Get AssetPath() -> (AssetPath:String), C:String="')") -> (ReturnValue))
   "On Clicked (NiagaraSystem)"() {is=title K2Node_ComponentBoundEvent}
   n1 = Append(A="import material_query_euw; material_query_euw.open_niagara_system(r'", B=Get NiagaraSystemPath() -> (NiagaraSystemPath:String), C="')")
   > ExecutePythonCommand(PythonCommand=@n1)
   Event PreConstruct()
   > "SetText (Text)"(self=Get AssetNameText() -> (AssetNameText:Text Object Reference), InText="To Text (String)"(InString=Get AssetName() -> (AssetName:String)))
   > "SetText (Text)"(self=Get WhyText() -> (WhyText:Text Object Reference), InText="To Text (String)"(InString=Get Why() -> (Why:String)))
   n2 = Get AssetPath()
   n3 = Get AssetName()
   > "SetText (Text)"(self=Get AssetPathText() -> (AssetPathText:Text Object Reference), InText="To Text (String)"(InString=Replace(SourceString=@n2, From=@n3)))
   > n4 = Branch(Condition=Get NoResults() -> (NoResults:bool))
   > SetVisibility(self=Get NumericalData() -> (NumericalData:Vertical Box Object Reference), InVisibility=Collapsed)
   > SetVisibility(self=Get AssetNameText(), InVisibility=Collapsed)
   "SetText (Text)"(
    execute=@n4.else
    self=Get ConfidenceText() -> (ConfidenceText:Text Object Reference)
    InText="To Text (Float)"(
     Value=Get Confidence() -> (Confidence:double)
     MinimumIntegralDigits=1
     MaximumIntegralDigits=324
     MinimumFractionalDigits=2
     MaximumFractionalDigits=2
    )
   )
   > "SetText (Text)"(
    self=Get NodeCountText() -> (NodeCountText:Text Object Reference)
    InText="To Text (Float)"(
     Value=Get NodeCount() -> (NodeCount:double)
     MinimumIntegralDigits=1
     MaximumIntegralDigits=324
     MinimumFractionalDigits=0
     MaximumFractionalDigits=0
    )
   )
   > "SetText (Text)"(
    self=Get ParameterText() -> (ParameterText:Text Object Reference)
    InText="To Text (Float)"(
     Value=Get ParameterCount() -> (ParameterCount:double)
     MinimumIntegralDigits=1
     MaximumIntegralDigits=324
     MinimumFractionalDigits=0
     MaximumFractionalDigits=0
    )
   )
   PrintString(InString=@n1)
   PrintString(InString=Hello)
   Event Construct()
   Event Tick()
  ]}
 GetBrushColor {nodes [
   GetBrushColor() {is=title K2Node_FunctionEntry}
   > n1 = Branch(Condition=Get NoResults())
   > Return Node!(execute, ReturnValue:Linear Color Structure=(R=1,G=0,B=0,A=0.15))
   Return Node(execute=@n1.else, ReturnValue=(R=1,G=1,B=1,A=0.15))
  ]}
 Get_NiagaraSystem_Visibility {nodes [
   Get_NiagaraSystem_Visibility() {is=title K2Node_FunctionEntry}
   n1 = Get NiagaraSystemPath()
   > n2 = Branch(Condition="integer > integer"(A=Len(S=@n1)))
   > Return Node(ReturnValue:ESlateVisibility Enum=Visible)
   Return Node(execute=@n2.else, ReturnValue=Collapsed)
  ]}
]
variables [
 AssetName {type=string, default=""}
 AssetPath {type=string, default=""}
 Confidence {type=number}
 NiagaraSystemPath {type=string, default=""}
 NodeCount {type=number}
 NoResults {type=boolean}
 ParameterCount {type=number}
 Why {type=string, default=""}
]
widgets {
 info {rootWidgetClass=&WrapBox, widgetCount=28}
 widgets [
  WrapBox_0 {widgetClassPath=&WrapBox, widget=WrapBox_0}
  Border_1 {
   widgetClassPath=&Border
   parent=WrapBox_0
   widget=Border_1
   properties {brushColor=<rgba 0 0 0 0.1>, padding=<ltrb 5 5 5 2>}
   slot {class=WrapBoxSlot, properties {bFillEmptySpace=true, bForceNewLine=true, padding=<ltrb 5 5 5 5>}}
  }
  VerticalBox {widgetClassPath=&VerticalBox, parent=Border_1, widget=VerticalBox, slot {class=BorderSlot, properties {padding=<ltrb 5 5 5 2>}}}
  VerticalBox_2 {
   widgetClassPath=&VerticalBox
   parent=VerticalBox
   widget=VerticalBox_2
   properties {bOverride_Cursor=true}
   slot {class=VerticalBoxSlot, properties {padding=<ltrb 0 0 0 10>}}
  }
  HorizontalBox_3 {widgetClassPath=&HorizontalBox, parent=VerticalBox_2, widget=HorizontalBox_3, slot {class=VerticalBoxSlot}}
  AssetNameText {
   widgetClassPath=&TextBlock
   bIsVariable=true
   parent=HorizontalBox_3
   widget=AssetNameText
   properties {
    text=AssetNameText
    font {
     bForceMonospaced=false
     bMaterialIsStencil=false
     fontMaterial=null
     fontObject=&Roboto
     letterSpacing=0
     monospacedWidth=1
     size=10.5
     skewAmount=0
     typefaceFontName=Bold
     outlineSettings {
      bApplyOutlineToDropShadows=false
      bMiteredCorners=false
      bSeparateFillAlpha=false
      outlineMaterial=null
      outlineSize=0
      outlineColor=<rgba 0 0 0 1>
     }
    }
   }
   slot {class=HorizontalBoxSlot, properties {verticalAlignment=VAlign_Center, size {sizeRule=Fill, value=1}}}
  }
  HorizontalBox_4 {widgetClassPath=&HorizontalBox, parent=VerticalBox_2, widget=HorizontalBox_4, slot {class=VerticalBoxSlot}}
  AssetPathText {
   widgetClassPath=&TextBlock
   bIsVariable=true
   parent=HorizontalBox_4
   widget=AssetPathText
   properties {
    text=AssetPathText
    font {
     bForceMonospaced=false
     bMaterialIsStencil=false
     fontMaterial=null
     fontObject=&Roboto
     letterSpacing=0
     monospacedWidth=1
     size=9
     skewAmount=0
     typefaceFontName=Regular
     outlineSettings {
      bApplyOutlineToDropShadows=false
      bMiteredCorners=false
      bSeparateFillAlpha=false
      outlineMaterial=null
      outlineSize=0
      outlineColor=<rgba 0 0 0 1>
     }
    }
   }
   slot {class=HorizontalBoxSlot, properties {verticalAlignment=VAlign_Center, size {sizeRule=Fill, value=1}}}
  }
  HorizontalBox_5 {widgetClassPath=&HorizontalBox, parent=VerticalBox_2, widget=HorizontalBox_5, slot {class=VerticalBoxSlot}}
  WhyText {
   widgetClassPath=&TextBlock
   bIsVariable=true
   parent=HorizontalBox_5
   widget=WhyText
   properties {
    autoWrapText=true
    minDesiredWidth=500
    text=WhyText
    font {
     bForceMonospaced=false
     bMaterialIsStencil=false
     fontMaterial=null
     fontObject=&Roboto
     letterSpacing=0
     monospacedWidth=1
     size=9
     skewAmount=0
     typefaceFontName=Light
     outlineSettings {
      bApplyOutlineToDropShadows=false
      bMiteredCorners=false
      bSeparateFillAlpha=false
      outlineMaterial=null
      outlineSize=0
      outlineColor=<rgba 0 0 0 1>
     }
    }
   }
   slot {class=HorizontalBoxSlot, properties {verticalAlignment=VAlign_Center, size {sizeRule=Fill, value=1}}}
  }
  Border_6 {
   widgetClassPath=&Border
   parent=VerticalBox
   widget=Border_6
   properties {brushColor=<rgba 0 0 0 0.4>}
   slot {class=VerticalBoxSlot, properties {padding=<ltrb 6 4 6 4>}}
  }
  NumericalData {widgetClassPath=&VerticalBox, bIsVariable=true, parent=Border_6, widget=NumericalData, slot {class=BorderSlot}}
  HorizontalBox_2 {
   widgetClassPath=&HorizontalBox
   parent=NumericalData
   widget=HorizontalBox_2
   properties {bOverride_Cursor=true}
   slot {class=VerticalBoxSlot, properties {horizontalAlignment=HAlign_Left}}
  }
  HorizontalBox_6 {
   widgetClassPath=&HorizontalBox
   parent=HorizontalBox_2
   widget=HorizontalBox_6
   slot {class=HorizontalBoxSlot, properties {padding=<ltrb 4 2 30 0>}}
  }
  HorizontalBox_0 {
   widgetClassPath=&HorizontalBox
   parent=HorizontalBox_6
   widget=HorizontalBox_0
   slot {class=HorizontalBoxSlot, properties {padding=<ltrb 0 0 12 0>}}
  }
  ConfidenceLabel {
   widgetClassPath=&TextBlock
   parent=HorizontalBox_0
   widget=ConfidenceLabel
   properties {
    autoWrapText=true
    text=Score:
    toolTipText=How confident the AI is in this result
    font {
     bForceMonospaced=false
     bMaterialIsStencil=false
     fontMaterial=null
     fontObject=&Roboto
     letterSpacing=0
     monospacedWidth=1
     size=10.5
     skewAmount=0
     typefaceFontName=Italic
     outlineSettings {
      bApplyOutlineToDropShadows=false
      bMiteredCorners=false
      bSeparateFillAlpha=false
      outlineMaterial=null
      outlineSize=0
      outlineColor=<rgba 0 0 0 1>
     }
    }
   }
   slot {class=HorizontalBoxSlot, properties {verticalAlignment=VAlign_Center, padding=<ltrb 0 0 1 0>}}
  }
  ConfidenceText {
   widgetClassPath=&TextBlock
   bIsVariable=true
   parent=HorizontalBox_0
   widget=ConfidenceText
   properties {
    text=ConfidenceText
    toolTipText=How confident the AI is in this result
    font {
     bForceMonospaced=false
     bMaterialIsStencil=false
     fontMaterial=null
     fontObject=&Roboto
     letterSpacing=0
     monospacedWidth=1
     size=12
     skewAmount=0
     typefaceFontName=Bold Italic
     outlineSettings {
      bApplyOutlineToDropShadows=false
      bMiteredCorners=false
      bSeparateFillAlpha=false
      outlineMaterial=null
      outlineSize=0
      outlineColor=<rgba 0 0 0 1>
     }
    }
   }
   slot {class=HorizontalBoxSlot, properties {verticalAlignment=VAlign_Center, padding=<ltrb 0 0 15 0>}}
  }
  HorizontalBox_1 {
   widgetClassPath=&HorizontalBox
   parent=HorizontalBox_6
   widget=HorizontalBox_1
   slot {class=HorizontalBoxSlot, properties {padding=<ltrb 0 0 15 0>}}
  }
  NodeIcon {
   widgetClassPath=&Image
   bIsVariable=true
   parent=HorizontalBox_1
   widget=NodeIcon
   properties {
    toolTipText=Number of Nodes in Material
    brush {
     bIsDynamicallyLoaded=false
     drawAs=Image
     imageType=FullColor
     mirroring=NoMirror
     resourceName=None
     resourceObject=null
     tiling=NoTile
     imageSize=<24 24>
     margin=<ltrb 0 0 0 0>
     outlineSettings {
      bUseBrushTransparency=false
      roundingType=HalfHeightRadius
      width=0
      color {colorUseRule=UseColor_Specified, specifiedColor=<rgba 0 0 0 0>}
      cornerRadii=<0 0 0 0>
     }
     tintColor {colorUseRule=UseColor_Specified, specifiedColor=<rgba 1 1 1 1>}
     uVRegion {bIsValid=false, max=<0 0>, min=<0 0>}
    }
   }
   slot {class=HorizontalBoxSlot, properties {horizontalAlignment=HAlign_Center, verticalAlignment=VAlign_Center, padding=<ltrb 2 0 1 0>}}
  }
  NodeCountText {
   widgetClassPath=&TextBlock
   bIsVariable=true
   parent=HorizontalBox_1
   widget=NodeCountText
   properties {
    text=0.00
    toolTipText=Number of Nodes in Material
    font {
     bForceMonospaced=false
     bMaterialIsStencil=false
     fontMaterial=null
     fontObject=&Roboto
     letterSpacing=0
     monospacedWidth=1
     size=12
     skewAmount=0
     typefaceFontName=Bold Italic
     outlineSettings {
      bApplyOutlineToDropShadows=false
      bMiteredCorners=false
      bSeparateFillAlpha=false
      outlineMaterial=null
      outlineSize=0
      outlineColor=<rgba 0 0 0 1>
     }
    }
   }
   slot {class=HorizontalBoxSlot, properties {verticalAlignment=VAlign_Center}}
  }
  HorizontalBox {
   widgetClassPath=&HorizontalBox
   parent=HorizontalBox_6
   widget=HorizontalBox
   slot {class=HorizontalBoxSlot, properties {padding=<ltrb 0 0 15 0>}}
  }
  ParameterIcon {
   widgetClassPath=&Image
   bIsVariable=true
   parent=HorizontalBox
   widget=ParameterIcon
   properties {
    toolTipText=Number of Parameters in Material
    brush {
     bIsDynamicallyLoaded=false
     drawAs=Image
     imageType=FullColor
     mirroring=NoMirror
     resourceName=None
     resourceObject=/AnimationLibrary/resources/icons/anglesrightsolid
     tiling=NoTile
     imageSize=<24 24>
     margin=<ltrb 0 0 0 0>
     outlineSettings {
      bUseBrushTransparency=false
      roundingType=HalfHeightRadius
      width=0
      color {colorUseRule=UseColor_Specified, specifiedColor=<rgba 0 0 0 0>}
      cornerRadii=<0 0 0 0>
     }
     tintColor {colorUseRule=UseColor_Specified, specifiedColor=<rgba 1 1 1 1>}
     uVRegion {bIsValid=false, max=<0 0>, min=<0 0>}
    }
   }
   slot {class=HorizontalBoxSlot, properties {horizontalAlignment=HAlign_Center, verticalAlignment=VAlign_Center, padding=<ltrb 2 0 1 0>}}
  }
  ParameterText {
   widgetClassPath=&TextBlock
   bIsVariable=true
   parent=HorizontalBox
   widget=ParameterText
   properties {
    text=0.00
    toolTipText=Number of Parameters in Material
    font {
     bForceMonospaced=false
     bMaterialIsStencil=false
     fontMaterial=null
     fontObject=&Roboto
     letterSpacing=0
     monospacedWidth=1
     size=12
     skewAmount=0
     typefaceFontName=Bold Italic
     outlineSettings {
      bApplyOutlineToDropShadows=false
      bMiteredCorners=false
      bSeparateFillAlpha=false
      outlineMaterial=null
      outlineSize=0
      outlineColor=<rgba 0 0 0 1>
     }
    }
   }
   slot {class=HorizontalBoxSlot, properties {verticalAlignment=VAlign_Center}}
  }
  HorizontalBox_7 {
   widgetClassPath=&HorizontalBox
   parent=HorizontalBox_2
   widget=HorizontalBox_7
   slot {class=HorizontalBoxSlot, properties {padding=<ltrb 20 0 20 0>}}
  }
  OpenButton {
   widgetClassPath=&EditorUtilityButton
   parent=HorizontalBox_7
   widget=OpenButton
   properties {
    toolTipText=Open the Material
    widgetStyle {
     clickedSlateSound {resourceObject=null}
     disabled {
      bIsDynamicallyLoaded=false
      drawAs=RoundedBox
      imageType=NoImage
      mirroring=NoMirror
      resourceName=None
      resourceObject=null
      tiling=NoTile
      imageSize=<0 0>
      margin=<ltrb 0 0 0 0>
      outlineSettings {
       bUseBrushTransparency=false
       roundingType=FixedRadius
       width=1
       color {colorUseRule=UseColor_ColorTable, specifiedColor=<rgba 0 0 0 0>}
       cornerRadii=<4 4 4 4>
      }
      tintColor {colorUseRule=UseColor_Specified, specifiedColor=<rgba 0.0395 0.0395 0.0395 1>}
      uVRegion {bIsValid=false, max=<0 0>, min=<0 0>}
     }
     disabledForeground {colorUseRule=UseColor_ColorTable, specifiedColor=<rgba 0 0 0 0>}
     hovered {
      bIsDynamicallyLoaded=false
      drawAs=RoundedBox
      imageType=NoImage
      mirroring=NoMirror
      resourceName=None
      resourceObject=null
      tiling=NoTile
      imageSize=<32 32>
      margin=<ltrb 0 0 0 0>
      outlineSettings {
       bUseBrushTransparency=false
       roundingType=FixedRadius
       width=1
       color {colorUseRule=UseColor_ColorTable, specifiedColor=<rgba 0 0 0 0>}
       cornerRadii=<4 4 4 4>
      }
      tintColor {colorUseRule=UseColor_Specified, specifiedColor=<rgba 0.2195 0.6514 0.9823 0.65>}
      uVRegion {bIsValid=false, max=<0 0>, min=<0 0>}
     }
     hoveredForeground {colorUseRule=UseColor_ColorTable, specifiedColor=<rgba 0 0 0 0>}
     hoveredSlateSound {resourceObject=null}
     normal {
      bIsDynamicallyLoaded=false
      drawAs=RoundedBox
      imageType=NoImage
      mirroring=NoMirror
      resourceName=None
      resourceObject=null
      tiling=NoTile
      imageSize=<32 32>
      margin=<ltrb 0 0 0 0>
      outlineSettings {
       bUseBrushTransparency=false
       roundingType=FixedRadius
       width=1
       color {colorUseRule=UseColor_ColorTable, specifiedColor=<rgba 0 0 0 0>}
       cornerRadii=<4 4 4 4>
      }
      tintColor {colorUseRule=UseColor_Specified, specifiedColor=<rgba 0.2195 0.6514 0.9823 0.302>}
      uVRegion {bIsValid=false, max=<0 0>, min=<0 0>}
     }
     normalForeground {colorUseRule=UseColor_ColorTable, specifiedColor=<rgba 0 0 0 0>}
     normalPadding=<ltrb 12 1.5 12 1.5>
     pressed {
      bIsDynamicallyLoaded=false
      drawAs=RoundedBox
      imageType=NoImage
      mirroring=NoMirror
      resourceName=None
      resourceObject=null
      tiling=NoTile
      imageSize=<32 32>
      margin=<ltrb 0 0 0 0>
      outlineSettings {
       bUseBrushTransparency=false
       roundingType=FixedRadius
       width=1
       color {colorUseRule=UseColor_ColorTable, specifiedColor=<rgba 0 0 0 0>}
       cornerRadii=<4 4 4 4>
      }
      tintColor {colorUseRule=UseColor_Specified, specifiedColor=<rgba 0.2195 0.6514 0.9823 0.302>}
      uVRegion {bIsValid=false, max=<0 0>, min=<0 0>}
     }
     pressedForeground {colorUseRule=UseColor_ColorTable, specifiedColor=<rgba 0 0 0 0>}
     pressedPadding=<ltrb 12 2.5 12 0.5>
     pressedSlateSound {resourceObject=null}
    }
   }
   slot {class=HorizontalBoxSlot, properties {padding=<ltrb 0 0 30 0>}}
  }
  Open {
   widgetClassPath=&TextBlock
   parent=OpenButton
   widget=Open
   properties {
    text=Material
    font {
     bForceMonospaced=false
     bMaterialIsStencil=false
     fontMaterial=null
     fontObject=&Roboto
     letterSpacing=0
     monospacedWidth=1
     size=11.25
     skewAmount=0
     typefaceFontName=Bold
     outlineSettings {
      bApplyOutlineToDropShadows=false
      bMiteredCorners=false
      bSeparateFillAlpha=false
      outlineMaterial=null
      outlineSize=0
      outlineColor=<rgba 0 0 0 1>
     }
    }
   }
   slot {class=ButtonSlot, properties {padding=<ltrb 20 0 20 0>}}
  }
  NiagaraSystem {
   widgetClassPath=&EditorUtilityButton
   parent=HorizontalBox_7
   widget=NiagaraSystem
   properties {
    toolTipText=Open the Niagara System
    widgetStyle {
     clickedSlateSound {resourceObject=null}
     disabled {
      bIsDynamicallyLoaded=false
      drawAs=RoundedBox
      imageType=NoImage
      mirroring=NoMirror
      resourceName=None
      resourceObject=null
      tiling=NoTile
      imageSize=<0 0>
      margin=<ltrb 0 0 0 0>
      outlineSettings {
       bUseBrushTransparency=false
       roundingType=FixedRadius
       width=1
       color {colorUseRule=UseColor_ColorTable, specifiedColor=<rgba 0 0 0 0>}
       cornerRadii=<4 4 4 4>
      }
      tintColor {colorUseRule=UseColor_Specified, specifiedColor=<rgba 0.0395 0.0395 0.0395 1>}
      uVRegion {bIsValid=false, max=<0 0>, min=<0 0>}
     }
     disabledForeground {colorUseRule=UseColor_ColorTable, specifiedColor=<rgba 0 0 0 0>}
     hovered {
      bIsDynamicallyLoaded=false
      drawAs=RoundedBox
      imageType=NoImage
      mirroring=NoMirror
      resourceName=None
      resourceObject=null
      tiling=NoTile
      imageSize=<32 32>
      margin=<ltrb 0 0 0 0>
      outlineSettings {
       bUseBrushTransparency=false
       roundingType=FixedRadius
       width=1
       color {colorUseRule=UseColor_ColorTable, specifiedColor=<rgba 0 0 0 0>}
       cornerRadii=<4 4 4 4>
      }
      tintColor {colorUseRule=UseColor_Specified, specifiedColor=<rgba 0.2195 0.6514 0.9823 0.65>}
      uVRegion {bIsValid=false, max=<0 0>, min=<0 0>}
     }
     hoveredForeground {colorUseRule=UseColor_ColorTable, specifiedColor=<rgba 0 0 0 0>}
     hoveredSlateSound {resourceObject=null}
     normal {
      bIsDynamicallyLoaded=false
      drawAs=RoundedBox
      imageType=NoImage
      mirroring=NoMirror
      resourceName=None
      resourceObject=null
      tiling=NoTile
      imageSize=<32 32>
      margin=<ltrb 0 0 0 0>
      outlineSettings {
       bUseBrushTransparency=false
       roundingType=FixedRadius
       width=1
       color {colorUseRule=UseColor_ColorTable, specifiedColor=<rgba 0 0 0 0>}
       cornerRadii=<4 4 4 4>
      }
      tintColor {colorUseRule=UseColor_Specified, specifiedColor=<rgba 0.2195 0.6514 0.9823 0.3>}
      uVRegion {bIsValid=false, max=<0 0>, min=<0 0>}
     }
     normalForeground {colorUseRule=UseColor_ColorTable, specifiedColor=<rgba 0 0 0 0>}
     normalPadding=<ltrb 12 1.5 12 1.5>
     pressed {
      bIsDynamicallyLoaded=false
      drawAs=RoundedBox
      imageType=NoImage
      mirroring=NoMirror
      resourceName=None
      resourceObject=null
      tiling=NoTile
      imageSize=<32 32>
      margin=<ltrb 0 0 0 0>
      outlineSettings {
       bUseBrushTransparency=false
       roundingType=FixedRadius
       width=1
       color {colorUseRule=UseColor_ColorTable, specifiedColor=<rgba 0 0 0 0>}
       cornerRadii=<4 4 4 4>
      }
      tintColor {colorUseRule=UseColor_Specified, specifiedColor=<rgba 0.2195 0.6514 0.9823 0.302>}
      uVRegion {bIsValid=false, max=<0 0>, min=<0 0>}
     }
     pressedForeground {colorUseRule=UseColor_ColorTable, specifiedColor=<rgba 0 0 0 0>}
     pressedPadding=<ltrb 12 2.5 12 0.5>
     pressedSlateSound {resourceObject=null}
    }
   }
   slot {class=HorizontalBoxSlot}
  }
  NiagaraText {
   widgetClassPath=&TextBlock
   parent=NiagaraSystem
   widget=NiagaraText
   properties {
    text=Niagara System
    font {
     bForceMonospaced=false
     bMaterialIsStencil=false
     fontMaterial=null
     fontObject=&Roboto
     letterSpacing=0
     monospacedWidth=1
     size=11.25
     skewAmount=0
     typefaceFontName=Bold
     outlineSettings {
      bApplyOutlineToDropShadows=false
      bMiteredCorners=false
      bSeparateFillAlpha=false
      outlineMaterial=null
      outlineSize=0
      outlineColor=<rgba 0 0 0 1>
     }
    }
   }
   slot {class=ButtonSlot, properties {padding=<ltrb 20 0 20 0>}}
  }
 ]
}
