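# Shadernator editor utility widget that helps artists find materials by text query, Niagara renderer, blend mode and optional advanced filters.
# Search shows a loading label, briefly delays, then passes the inputs as strings to Python's material_query_euw.search_materials_delimited.
# The returned string is split by ";;;" and "|" into EUW_MatchRow widgets, with zero confidence flagging no results.
# A checkbox toggles advanced options: plugin, required nodes, reference image and expanded search.
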EditorUtilityWidgetBlueprint /DealershipSharedLibrary/ProductionTools/EUW_MaterialDiscovery
graphs [
 EventGraph {nodes [
   "On Clicked (SearchButton)"() {is=title K2Node_ComponentBoundEvent}
   n1 = Get SearchText() -> (SearchText:Text Object Reference)
   > "SetText (Text)"(self=@n1, InText="LOCTEXT(\"Loading...\")")
   > Set bIsEnabled!(execute, bIsEnabled:bool=false, self:Widget Object Reference=@n1) -> (then, Output_Get:bool)
   > Delay!(execute, WorldContextObject:Object Reference, Duration=0.01) -> (then)
   > ClearChildren(self=Get ResultsContainer() -> (ResultsContainer:Vertical Box Object Reference))
   n2 = Get PreferredPluginInput() -> (PreferredPluginInput:Editor Utility Combo Box String Object Reference)
   n3 = "GetText (Text Box)"!(self:Text Box Object Reference=Get NeedsInput() -> (NeedsInput:Editor Utility Editable Text Box Object Reference)) -> (ReturnValue:Text)
   n4 = Get BlendModeInput() -> (BlendModeInput:Editor Utility Combo Box String Object Reference)
   n5 = Get UsageInput() -> (UsageInput:Editor Utility Combo Box String Object Reference)
   n6 = "GetText (Text Box)"(self=Get QueryInput() -> (QueryInput:Editor Utility Editable Text Box Object Reference))
   n7 = IsChecked(self=Get ExpandedSearchInput() -> (ExpandedSearchInput:Editor Utility Check Box Object Reference))
   n8 = "GetText (Text Box)"(self=Get ImageInput() -> (ImageInput:Editor Utility Editable Text Box Object Reference))
   > n9 = "Execute Python Command (Advanced)"!(
    execute
    PythonCommand:String=Append!(
     A="__import__('material_query_euw').search_materials_delimited(\""
     B="To String (Text)"(InText=@n6)
     C:String="\",\""
     D:String=GetSelectedOption!(self:"Combo Box (String) Object Reference"=@n5) -> (ReturnValue:String)
     E:String="\",\""
     F:String=GetSelectedOption(self=@n4)
     G:String="\",\""
     H:String="To String (Text)"(InText=@n3)
     I:String="\",\""
     J:String="To String (Text)"(InText=@n8)
     K:String="\",\""
     L:String=GetSelectedOption(self=@n2)
     M:String="\",\""
     N:String=SelectString(A=true, bPickA=@n7)
     O:String="\")"
    ) -> (ReturnValue)
    ExecutionMode:EPythonCommandExecutionMode Enum=EvaluateStatement
    FileExecutionScope:EPythonFileExecutionScope Enum=Private
   ) -> (then, CommandResult:String, LogOutput:Array of Python Log Output Entry Structures, ReturnValue:bool)
   > PrintString(InString=@n9)
   > n10 = For Each Loop(Array=ParseIntoArray(SourceString=Replace(SourceString=@n9, From='), Delimiter=";;;"))
   n11 = Get SearchText()
   "SetText (Text)"(execute=@n10.Completed, self=@n11, InText="LOCTEXT(\"Search\")")
   > Set bIsEnabled(bIsEnabled=true, self=@n11)
   n12 = ParseIntoArray(SourceString=@n10, Delimiter=|, CullEmptyStrings=false)
   n13 = String To Float(InString="Get (a copy)"(Array=@n12, Dimension 1=2))
   n14 = "Equal (Float)"(A=@n13)
   n15 = Create Widget!(
    execute:Exec=@n10.LoopBody
    WorldContextObject:Object Reference
    Class:User Widget Class Reference=&/EUW_MatchRow.EUW_MatchRow_C
    OwningPlayer:Player Controller Object Reference
    AssetName:String="Get (a copy)"(Array=@n12)
    AssetPath:String="Get (a copy)"(Array=@n12, Dimension 1=1)
    Confidence:double=@n13
    Why:String="Get (a copy)"(Array=@n12, Dimension 1=3)
    NiagaraSystemPath:String="Get (a copy)"(Array=@n12, Dimension 1=4)
    NodeCount:double=String To Float(InString="Get (a copy)"(Array=@n12, Dimension 1=5))
    ParameterCount:double=String To Float(InString="Get (a copy)"(Array=@n12, Dimension 1=6))
    NoResults:bool=@n14
   ) -> (then:Exec, ReturnValue:EUW Match Row Object Reference)
   > AddChild(self=Get ResultsContainer(), Content=@n15)
   > Branch(Condition=@n14)
   > PrintString(InString=No results)
   n16 = "On Check State Changed (AdvancedOptionsCheckbox)"!() -> (OutputDelegate, then, bIsChecked:bool) {is=title K2Node_ComponentBoundEvent}
   > n17 = Branch(Condition=@n16.bIsChecked)
   > SetVisibility(self=Get AdvancedOptions() -> (AdvancedOptions:Vertical Box Object Reference), InVisibility=Visible)
   SetVisibility(execute=@n17.else, self=Get AdvancedOptions(), InVisibility=Collapsed)
   Event PreConstruct()
   > SetVisibility(self=Get AdvancedOptions(), InVisibility=Collapsed)
   LogString(InString=@n10)
   Event Tick()
  ]}
 GetCheckedState {nodes [
   GetCheckedState() {is=title K2Node_FunctionEntry}
   > Return Node!(execute, ReturnValue:ECheckBoxState Enum=Unchecked)
  ]}
]
variables []
widgets {
 info {rootWidgetClass=&VerticalBox, widgetCount=42}
 widgets [
  VerticalBox_238 {widgetClassPath=&VerticalBox, widget=VerticalBox_238, properties {bOverride_Cursor=true}}
  HorizontalBox_0 {
   widgetClassPath=&HorizontalBox
   parent=VerticalBox_238
   widget=HorizontalBox_0
   slot {class=VerticalBoxSlot, properties {padding=<ltrb 5 0 0 0>}}
  }
  Shader {
   widgetClassPath=&TextBlock
   parent=HorizontalBox_0
   widget=Shader
   properties {
    minDesiredWidth=50
    text=Shader
    font {
     bForceMonospaced=false
     bMaterialIsStencil=false
     fontMaterial=null
     fontObject=&Roboto
     letterSpacing=0
     monospacedWidth=1
     size=15
     skewAmount=0
     typefaceFontName=Bold
     outlineSettings {
      bApplyOutlineToDropShadows=false
      bMiteredCorners=false
      bSeparateFillAlpha=false
      outlineMaterial=null
      outlineSize=0
      outlineColor=<rgba 0 0 0 1>
     }
    }
   }
   slot {class=HorizontalBoxSlot, properties {horizontalAlignment=HAlign_Left, verticalAlignment=VAlign_Center, padding=<ltrb 10 0 0 0>}}
  }
  nator {
   widgetClassPath=&TextBlock
   parent=HorizontalBox_0
   widget=nator
   properties {
    minDesiredWidth=50
    text=nator
    colorAndOpacity {colorUseRule=UseColor_Specified, specifiedColor=<rgba 0.2195 0.6514 0.9823 1>}
    font {
     bForceMonospaced=false
     bMaterialIsStencil=false
     fontMaterial=null
     fontObject=&Roboto
     letterSpacing=0
     monospacedWidth=1
     size=15
     skewAmount=0
     typefaceFontName=Bold
     outlineSettings {
      bApplyOutlineToDropShadows=false
      bMiteredCorners=false
      bSeparateFillAlpha=false
      outlineMaterial=null
      outlineSize=0
      outlineColor=<rgba 0 0 0 1>
     }
    }
   }
   slot {class=HorizontalBoxSlot, properties {horizontalAlignment=HAlign_Left, verticalAlignment=VAlign_Center, padding=<ltrb 0 0 10 0>}}
  }
  ShadernatorDescription {
   widgetClassPath=&TextBlock
   parent=HorizontalBox_0
   widget=ShadernatorDescription
   properties {
    text=Material Discovery Tool for Artists
    font {
     bForceMonospaced=false
     bMaterialIsStencil=false
     fontMaterial=null
     fontObject=&Roboto
     letterSpacing=0
     monospacedWidth=1
     size=9
     skewAmount=0
     typefaceFontName=Italic
     outlineSettings {
      bApplyOutlineToDropShadows=false
      bMiteredCorners=false
      bSeparateFillAlpha=false
      outlineMaterial=null
      outlineSize=0
      outlineColor=<rgba 0 0 0 1>
     }
    }
   }
   slot {
    class=HorizontalBoxSlot
    properties {horizontalAlignment=HAlign_Left, verticalAlignment=VAlign_Center, padding=<ltrb 90 0 0 0>, size {sizeRule=Fill, value=1}}
   }
  }
  HorizontalBox_4 {widgetClassPath=&HorizontalBox, parent=VerticalBox_238, widget=HorizontalBox_4, slot {class=VerticalBoxSlot}}
  HorizontalBox_5 {widgetClassPath=&HorizontalBox, parent=HorizontalBox_4, widget=HorizontalBox_5, slot {class=HorizontalBoxSlot}}
  Image_200 {
   widgetClassPath=&Image
   bIsVariable=true
   parent=HorizontalBox_5
   widget=Image_200
   properties {brush {
     bIsDynamicallyLoaded=false
     drawAs=Image
     imageType=FullColor
     mirroring=NoMirror
     resourceName=None
     resourceObject=/DealershipSharedLibrary/Python/Material_Query/T_DLR_shadernator_logo
     tiling=NoTile
     imageSize=<100 100>
     margin=<ltrb 0 0 0 0>
     outlineSettings {
      bUseBrushTransparency=false
      roundingType=HalfHeightRadius
      width=0
      color {colorUseRule=UseColor_Specified, specifiedColor=<rgba 0 0 0 0>}
      cornerRadii=<0 0 0 0>
     }
     tintColor {colorUseRule=UseColor_Specified, specifiedColor=<rgba 1 1 1 1>}
     uVRegion {bIsValid=false, max=<0 0>, min=<0 0>}
    }}
   slot {class=HorizontalBoxSlot, properties {horizontalAlignment=HAlign_Left, verticalAlignment=VAlign_Top}}
  }
  VerticalBox_0 {
   widgetClassPath=&VerticalBox
   parent=HorizontalBox_4
   widget=VerticalBox_0
   slot {class=HorizontalBoxSlot, properties {size {sizeRule=Fill, value=1}}}
  }
  HorizontalBox_61 {widgetClassPath=&HorizontalBox, parent=VerticalBox_0, widget=HorizontalBox_61, slot {class=VerticalBoxSlot}}
  Query {
   widgetClassPath=&TextBlock
   parent=HorizontalBox_61
   widget=Query
   properties {
    minDesiredWidth=120
    text=Query:
    toolTipText=Open ended AI prompt
    font {
     bForceMonospaced=false
     bMaterialIsStencil=false
     fontMaterial=null
     fontObject=&Roboto
     letterSpacing=0
     monospacedWidth=1
     size=9
     skewAmount=0
     typefaceFontName=Regular
     outlineSettings {
      bApplyOutlineToDropShadows=false
      bMiteredCorners=false
      bSeparateFillAlpha=false
      outlineMaterial=null
      outlineSize=0
      outlineColor=<rgba 0 0 0 1>
     }
    }
   }
   slot {class=HorizontalBoxSlot, properties {verticalAlignment=VAlign_Center, padding=<ltrb 5 0 0 0>}}
  }
  QueryInput {
   widgetClassPath=&EditorUtilityEditableTextBox
   bIsVariable=true
   parent=HorizontalBox_61
   widget=QueryInput
   properties {hintText="whispy smoke flipbook with normal maps, dissolving functionality, emissive color, etc..."}
   slot {class=HorizontalBoxSlot, properties {size {sizeRule=Fill, value=1}}}
  }
  HorizontalBox_716 {widgetClassPath=&HorizontalBox, parent=VerticalBox_0, widget=HorizontalBox_716, slot {class=VerticalBoxSlot}}
  NiagaraRenderer {
   widgetClassPath=&TextBlock
   parent=HorizontalBox_716
   widget=NiagaraRenderer
   properties {
    minDesiredWidth=120
    text=Niagara Renderer:
    toolTipText=Renderer usage in Niagara
    font {
     bForceMonospaced=false
     bMaterialIsStencil=false
     fontMaterial=null
     fontObject=&Roboto
     letterSpacing=0
     monospacedWidth=1
     size=9
     skewAmount=0
     typefaceFontName=Regular
     outlineSettings {
      bApplyOutlineToDropShadows=false
      bMiteredCorners=false
      bSeparateFillAlpha=false
      outlineMaterial=null
      outlineSize=0
      outlineColor=<rgba 0 0 0 1>
     }
    }
   }
   slot {class=HorizontalBoxSlot, properties {verticalAlignment=VAlign_Center, padding=<ltrb 5 0 0 0>}}
  }
  UsageInput {
   widgetClassPath=/Script/Blutility.EditorUtilityComboBoxString
   bIsVariable=true
   parent=HorizontalBox_716
   widget=UsageInput
   properties {
    selectedOption=sprite
    contentPadding=<ltrb 0 0 0 0>
    defaultOptions [sprite,ribbon,mesh]
    font {
     bForceMonospaced=false
     bMaterialIsStencil=false
     fontMaterial=null
     fontObject=&Roboto
     letterSpacing=0
     monospacedWidth=1
     size=9
     skewAmount=0
     typefaceFontName=Regular
     outlineSettings {
      bApplyOutlineToDropShadows=false
      bMiteredCorners=false
      bSeparateFillAlpha=false
      outlineMaterial=null
      outlineSize=0
      outlineColor=<rgba 0 0 0 1>
     }
    }
    foregroundColor {colorUseRule=UseColor_Specified, specifiedColor=<rgba 0.5271 0.5271 0.5271 1>}
   }
   slot {class=HorizontalBoxSlot, properties {size {sizeRule=Fill, value=1}}}
  }
  HorizontalBox_6 {widgetClassPath=&HorizontalBox, parent=VerticalBox_0, widget=HorizontalBox_6, slot {class=VerticalBoxSlot}}
  BlendMode {
   widgetClassPath=&TextBlock
   parent=HorizontalBox_6
   widget=BlendMode
   properties {
    minDesiredWidth=120
    text=Blend Mode:
    toolTipText=Material Blend Mode
    font {
     bForceMonospaced=false
     bMaterialIsStencil=false
     fontMaterial=null
     fontObject=&Roboto
     letterSpacing=0
     monospacedWidth=1
     size=9
     skewAmount=0
     typefaceFontName=Regular
     outlineSettings {
      bApplyOutlineToDropShadows=false
      bMiteredCorners=false
      bSeparateFillAlpha=false
      outlineMaterial=null
      outlineSize=0
      outlineColor=<rgba 0 0 0 1>
     }
    }
   }
   slot {class=HorizontalBoxSlot, properties {verticalAlignment=VAlign_Center, padding=<ltrb 5 0 0 0>}}
  }
  BlendModeInput {
   widgetClassPath=&EditorUtilityComboBoxString
   bIsVariable=true
   parent=HorizontalBox_6
   widget=BlendModeInput
   properties {
    selectedOption=Translucent
    contentPadding=<ltrb 0 0 0 0>
    defaultOptions [translucent, opaque, masked, additive, alpha composite, other]
    font {
     bForceMonospaced=false
     bMaterialIsStencil=false
     fontMaterial=null
     fontObject=&Roboto
     letterSpacing=0
     monospacedWidth=1
     size=9
     skewAmount=0
     typefaceFontName=Regular
     outlineSettings {
      bApplyOutlineToDropShadows=false
      bMiteredCorners=false
      bSeparateFillAlpha=false
      outlineMaterial=null
      outlineSize=0
      outlineColor=<rgba 0 0 0 1>
     }
    }
    foregroundColor {colorUseRule=UseColor_Specified, specifiedColor=<rgba 0.5271 0.5271 0.5271 1>}
   }
   slot {class=HorizontalBoxSlot, properties {size {sizeRule=Fill, value=1}}}
  }
  Border_98 {
   widgetClassPath=&Border
   parent=VerticalBox_0
   widget=Border_98
   properties {brushColor=<rgba 0 0 0 0.4>, padding=<ltrb 0 5 0 5>}
   slot {class=VerticalBoxSlot}
  }
  OptionalInputs {
   widgetClassPath=&VerticalBox
   parent=Border_98
   widget=OptionalInputs
   slot {class=BorderSlot, properties {padding=<ltrb 0 5 0 5>}}
  }
  HorizontalBox_7 {
   widgetClassPath=&HorizontalBox
   parent=OptionalInputs
   widget=HorizontalBox_7
   slot {class=VerticalBoxSlot, properties {padding=<ltrb 5 0 0 0>}}
  }
  OptionalInputsLabel {
   widgetClassPath=&TextBlock
   parent=HorizontalBox_7
   widget=OptionalInputsLabel
   properties {
    minDesiredWidth=120
    text=Advanced Options
    toolTipText=Technical and project specific options for advanced queries
    font {
     bForceMonospaced=false
     bMaterialIsStencil=false
     fontMaterial=null
     fontObject=&Roboto
     letterSpacing=0
     monospacedWidth=1
     size=9
     skewAmount=0
     typefaceFontName=Bold
     outlineSettings {
      bApplyOutlineToDropShadows=false
      bMiteredCorners=false
      bSeparateFillAlpha=false
      outlineMaterial=null
      outlineSize=0
      outlineColor=<rgba 0 0 0 1>
     }
    }
   }
   slot {class=HorizontalBoxSlot, properties {verticalAlignment=VAlign_Center}}
  }
  AdvancedOptionsCheckbox {
   widgetClassPath=&EditorUtilityCheckBox
   bIsVariable=true
   parent=HorizontalBox_7
   widget=AdvancedOptionsCheckbox
   slot {class=HorizontalBoxSlot, properties {size {sizeRule=Fill, value=1}}}
  }
  AdvancedOptions {widgetClassPath=&VerticalBox, bIsVariable=true, parent=OptionalInputs, widget=AdvancedOptions, slot {class=VerticalBoxSlot}}
  HorizontalBox_2 {widgetClassPath=&HorizontalBox, parent=AdvancedOptions, widget=HorizontalBox_2, slot {class=VerticalBoxSlot}}
  PreferredPlugin {
   widgetClassPath=&TextBlock
   parent=HorizontalBox_2
   widget=PreferredPlugin
   properties {
    minDesiredWidth=120
    text=Preferred Plugin:
    toolTipText=Look for materials used in this plugin
    font {
     bForceMonospaced=false
     bMaterialIsStencil=false
     fontMaterial=null
     fontObject=&Roboto
     letterSpacing=0
     monospacedWidth=1
     size=9
     skewAmount=0
     typefaceFontName=Regular
     outlineSettings {
      bApplyOutlineToDropShadows=false
      bMiteredCorners=false
      bSeparateFillAlpha=false
      outlineMaterial=null
      outlineSize=0
      outlineColor=<rgba 0 0 0 1>
     }
    }
   }
   slot {class=HorizontalBoxSlot, properties {verticalAlignment=VAlign_Center, padding=<ltrb 5 0 0 0>}}
  }
  PreferredPluginInput {
   widgetClassPath=&EditorUtilityComboBoxString
   bIsVariable=true
   parent=HorizontalBox_2
   widget=PreferredPluginInput
   properties {
    contentPadding=<ltrb 0 0 0 0>
    defaultOptions [
     ""
     /DealershipSharedLibrary/
     /VehicleCosmetics/Boosts/
     /VehicleCosmetics/DriftTrail/
     /DealershipExplosions/
     /Game/NeverCook/FX/
     /Game/Effects/Fort_Effects/
    ]
    font {
     bForceMonospaced=false
     bMaterialIsStencil=false
     fontMaterial=null
     fontObject=&Roboto
     letterSpacing=0
     monospacedWidth=1
     size=9
     skewAmount=0
     typefaceFontName=Regular
     outlineSettings {
      bApplyOutlineToDropShadows=false
      bMiteredCorners=false
      bSeparateFillAlpha=false
      outlineMaterial=null
      outlineSize=0
      outlineColor=<rgba 0 0 0 1>
     }
    }
    foregroundColor {colorUseRule=UseColor_Specified, specifiedColor=<rgba 0.5271 0.5271 0.5271 1>}
   }
   slot {class=HorizontalBoxSlot, properties {size {sizeRule=Fill, value=1}}}
  }
  HorizontalBox {widgetClassPath=&HorizontalBox, parent=AdvancedOptions, widget=HorizontalBox, slot {class=VerticalBoxSlot}}
  Needs {
   widgetClassPath=&TextBlock
   parent=HorizontalBox
   widget=Needs
   properties {
    minDesiredWidth=120
    text=Required Nodes:
    toolTipText=Specific material nodes you require
    font {
     bForceMonospaced=false
     bMaterialIsStencil=false
     fontMaterial=null
     fontObject=&Roboto
     letterSpacing=0
     monospacedWidth=1
     size=9
     skewAmount=0
     typefaceFontName=Regular
     outlineSettings {
      bApplyOutlineToDropShadows=false
      bMiteredCorners=false
      bSeparateFillAlpha=false
      outlineMaterial=null
      outlineSize=0
      outlineColor=<rgba 0 0 0 1>
     }
    }
   }
   slot {class=HorizontalBoxSlot, properties {verticalAlignment=VAlign_Center, padding=<ltrb 5 0 0 0>}}
  }
  NeedsInput {
   widgetClassPath=&EditorUtilityEditableTextBox
   bIsVariable=true
   parent=HorizontalBox
   widget=NeedsInput
   properties {hintText="Param2D, DynamicParameter, Smoothstep, MF_FX_EmissiveAdjustments, etc..."}
   slot {class=HorizontalBoxSlot, properties {size {sizeRule=Fill, value=1}}}
  }
  HorizontalBox_1 {widgetClassPath=&HorizontalBox, parent=AdvancedOptions, widget=HorizontalBox_1, slot {class=VerticalBoxSlot}}
  ReferenceImage {
   widgetClassPath=&TextBlock
   parent=HorizontalBox_1
   widget=ReferenceImage
   properties {
    minDesiredWidth=120
    text=Reference Image:
    toolTipText=Filepath to reference image for comparison
    font {
     bForceMonospaced=false
     bMaterialIsStencil=false
     fontMaterial=null
     fontObject=&Roboto
     letterSpacing=0
     monospacedWidth=1
     size=9
     skewAmount=0
     typefaceFontName=Regular
     outlineSettings {
      bApplyOutlineToDropShadows=false
      bMiteredCorners=false
      bSeparateFillAlpha=false
      outlineMaterial=null
      outlineSize=0
      outlineColor=<rgba 0 0 0 1>
     }
    }
   }
   slot {class=HorizontalBoxSlot, properties {verticalAlignment=VAlign_Center, padding=<ltrb 5 0 0 0>}}
  }
  ImageInput {
   widgetClassPath=&EditorUtilityEditableTextBox
   bIsVariable=true
   parent=HorizontalBox_1
   widget=ImageInput
   properties {hintText=C:/Users/Artist/Desktop/smokeflipbook.jpg}
   slot {class=HorizontalBoxSlot, properties {size {sizeRule=Fill, value=1}}}
  }
  HorizontalBox_3 {widgetClassPath=&HorizontalBox, parent=AdvancedOptions, widget=HorizontalBox_3, slot {class=VerticalBoxSlot}}
  ExpandedSearch {
   widgetClassPath=&TextBlock
   parent=HorizontalBox_3
   widget=ExpandedSearch
   properties {
    minDesiredWidth=120
    text=Expanded Search:
    toolTipText="Scans 3x more materials (increses cost but gives more accurate results)"
    font {
     bForceMonospaced=false
     bMaterialIsStencil=false
     fontMaterial=null
     fontObject=&Roboto
     letterSpacing=0
     monospacedWidth=1
     size=9
     skewAmount=0
     typefaceFontName=Regular
     outlineSettings {
      bApplyOutlineToDropShadows=false
      bMiteredCorners=false
      bSeparateFillAlpha=false
      outlineMaterial=null
      outlineSize=0
      outlineColor=<rgba 0 0 0 1>
     }
    }
   }
   slot {class=HorizontalBoxSlot, properties {verticalAlignment=VAlign_Center, padding=<ltrb 5 0 0 0>}}
  }
  ExpandedSearchInput {
   widgetClassPath=&EditorUtilityCheckBox
   bIsVariable=true
   parent=HorizontalBox_3
   widget=ExpandedSearchInput
   slot {class=HorizontalBoxSlot}
  }
  SearchButton {
   widgetClassPath=&EditorUtilityButton
   parent=VerticalBox_238
   widget=SearchButton
   properties {
    toolTipText=Search Button
    widgetStyle {
     clickedSlateSound {resourceObject=null}
     disabled {
      bIsDynamicallyLoaded=false
      drawAs=RoundedBox
      imageType=NoImage
      mirroring=NoMirror
      resourceName=None
      resourceObject=null
      tiling=NoTile
      imageSize=<0 0>
      margin=<ltrb 0 0 0 0>
      outlineSettings {
       bUseBrushTransparency=false
       roundingType=FixedRadius
       width=1
       color {colorUseRule=UseColor_ColorTable, specifiedColor=<rgba 0 0 0 0>}
       cornerRadii=<4 4 4 4>
      }
      tintColor {colorUseRule=UseColor_Specified, specifiedColor=<rgba 0.0395 0.0395 0.0395 1>}
      uVRegion {bIsValid=false, max=<0 0>, min=<0 0>}
     }
     disabledForeground {colorUseRule=UseColor_ColorTable, specifiedColor=<rgba 0 0 0 0>}
     hovered {
      bIsDynamicallyLoaded=false
      drawAs=RoundedBox
      imageType=NoImage
      mirroring=NoMirror
      resourceName=None
      resourceObject=null
      tiling=NoTile
      imageSize=<32 32>
      margin=<ltrb 0 0 0 0>
      outlineSettings {
       bUseBrushTransparency=false
       roundingType=FixedRadius
       width=1
       color {colorUseRule=UseColor_ColorTable, specifiedColor=<rgba 0 0 0 0>}
       cornerRadii=<4 4 4 4>
      }
      tintColor {colorUseRule=UseColor_Specified, specifiedColor=<rgba 0.0953 0.0953 0.0953 1>}
      uVRegion {bIsValid=false, max=<0 0>, min=<0 0>}
     }
     hoveredForeground {colorUseRule=UseColor_ColorTable, specifiedColor=<rgba 0 0 0 0>}
     hoveredSlateSound {resourceObject=null}
     normal {
      bIsDynamicallyLoaded=false
      drawAs=RoundedBox
      imageType=NoImage
      mirroring=NoMirror
      resourceName=None
      resourceObject=null
      tiling=NoTile
      imageSize=<32 32>
      margin=<ltrb 0 0 0 0>
      outlineSettings {
       bUseBrushTransparency=false
       roundingType=FixedRadius
       width=1
       color {colorUseRule=UseColor_ColorTable, specifiedColor=<rgba 0 0 0 0>}
       cornerRadii=<4 4 4 4>
      }
      tintColor {colorUseRule=UseColor_Specified, specifiedColor=<rgba 0.0395 0.0395 0.0395 1>}
      uVRegion {bIsValid=false, max=<0 0>, min=<0 0>}
     }
     normalForeground {colorUseRule=UseColor_ColorTable, specifiedColor=<rgba 0 0 0 0>}
     normalPadding=<ltrb 12 1.5 12 1.5>
     pressed {
      bIsDynamicallyLoaded=false
      drawAs=RoundedBox
      imageType=NoImage
      mirroring=NoMirror
      resourceName=None
      resourceObject=null
      tiling=NoTile
      imageSize=<32 32>
      margin=<ltrb 0 0 0 0>
      outlineSettings {
       bUseBrushTransparency=false
       roundingType=FixedRadius
       width=1
       color {colorUseRule=UseColor_ColorTable, specifiedColor=<rgba 0 0 0 0>}
       cornerRadii=<4 4 4 4>
      }
      tintColor {colorUseRule=UseColor_Specified, specifiedColor=<rgba 0.0284 0.0284 0.0284 1>}
      uVRegion {bIsValid=false, max=<0 0>, min=<0 0>}
     }
     pressedForeground {colorUseRule=UseColor_ColorTable, specifiedColor=<rgba 0 0 0 0>}
     pressedPadding=<ltrb 12 2.5 12 0.5>
     pressedSlateSound {resourceObject=null}
    }
   }
   slot {class=VerticalBoxSlot}
  }
  SearchText {
   widgetClassPath=&TextBlock
   bIsVariable=true
   parent=SearchButton
   widget=SearchText
   properties {
    text=Search
    colorAndOpacity {colorUseRule=UseColor_Foreground, specifiedColor=<rgba 1 1 1 1>}
    font {
     bForceMonospaced=false
     bMaterialIsStencil=false
     fontMaterial=null
     fontObject=&Roboto
     letterSpacing=0
     monospacedWidth=1
     size=15
     skewAmount=0
     typefaceFontName=Bold
     outlineSettings {
      bApplyOutlineToDropShadows=false
      bMiteredCorners=false
      bSeparateFillAlpha=false
      outlineMaterial=null
      outlineSize=0
      outlineColor=<rgba 0 0 0 1>
     }
    }
   }
   slot {class=ButtonSlot, properties {padding=<ltrb 0 3 0 3>}}
  }
  Results {
   widgetClassPath=&TextBlock
   parent=VerticalBox_238
   widget=Results
   properties {
    text=Results:
    font {
     bForceMonospaced=false
     bMaterialIsStencil=false
     fontMaterial=null
     fontObject=&Roboto
     letterSpacing=0
     monospacedWidth=1
     size=12
     skewAmount=0
     typefaceFontName=Bold
     outlineSettings {
      bApplyOutlineToDropShadows=false
      bMiteredCorners=false
      bSeparateFillAlpha=false
      outlineMaterial=null
      outlineSize=0
      outlineColor=<rgba 0 0 0 1>
     }
    }
   }
   slot {class=VerticalBoxSlot, properties {horizontalAlignment=HAlign_Left, verticalAlignment=VAlign_Center, padding=<ltrb 5 0 0 0>}}
  }
  ResultsScroll {
   widgetClassPath=&EditorUtilityScrollBox
   parent=VerticalBox_238
   widget=ResultsScroll
   slot {class=VerticalBoxSlot, properties {size {sizeRule=Fill, value=1}}}
  }
  Border_143 {
   widgetClassPath=&Border
   parent=ResultsScroll
   widget=Border_143
   properties {brushColor=<rgba 0 0 0 0.5>}
   slot {class=ScrollBoxSlot}
  }
  ResultsContainer {widgetClassPath=&VerticalBox, bIsVariable=true, parent=Border_143, widget=ResultsContainer, slot {class=BorderSlot}}
 ]
}
